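# Scratchpad example material for skeletal and deformer-driven meshes, demonstrating pre-skinned triplanar projection with animated storm noise.
# MF_TriplanarVirtualBox produces X, Y and Z UV sets plus Y and Z blend masks, with parameters for tile, offset, depth and blend contrast.
# Named reroutes carry UVs into three MF_Polar and MF_StormNoise chains, one per axis, and two nested lerps form the triplanar blend.
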Material /Game/Global/Materials/MaterialGym/Materials/MaterialElements/UVs/ScratchPad/M_PreskinnedTriplanar_example
graph {
 properties {bAutomaticallySetUsageInEditor=false, bUsedWithMeshDeformer=true, bUsedWithSkeletalMesh=true, usedShadingModels=Default Lit}
 nodes [
  Constant()
  LinearInterpolate(
   A=LinearInterpolate(A=NamedRerouteUsage(), B=NamedRerouteUsage(), Alpha=NamedRerouteUsage())
   B=NamedRerouteUsage()
   Alpha=NamedRerouteUsage()
  )
  n1 = MaterialFunctionCall!(
   TVB_Texture
   TVB_UV_Tiling=ScalarParameter() {defaultValue=2, parameterName="Tile (S)"}
   TVB_UV_Offset=ComponentMask(None=VectorParameter() {parameterName=Offset, defaultValue=<rgba 0.004 0.466 0 1>}.RGB) {g=true, r=true}
   TVB_CenterPointOffset
   TVB_Depth=ScalarParameter() {parameterName="Depth (S)"}
   TVB_BlendContrast=ScalarParameter() {defaultValue=10, parameterName=BlendContrast}
   TVB_UseLocalCameraVector
  ) {materialFunction=/Game/Global/Materials/MaterialGym/Materials/MaterialElements/UVs/PreskinnedVirtualBox/MF_TriplanarVirtualBox}
  NamedRerouteDeclaration(None=@n1.UV X) {name=UVX, nodeColor=<rgba 0 1 0.4375 1>}
  NamedRerouteDeclaration(None=@n1.UV Y) {name=UVY, nodeColor=<rgba 1 0.1172 0 1>}
  NamedRerouteDeclaration(None=@n1.UV Z) {name=UVZ, nodeColor=<rgba 0 0.9531 1 1>}
  NamedRerouteDeclaration(None=@n1.UV Blend Y) {name=UVBlendY, nodeColor=<rgba 1 0.7266 0 1>}
  NamedRerouteDeclaration(None=@n1.UV Blend Z) {name=UVBlendZ, nodeColor=<rgba 0.4062 0 1 1>}
  n2 = NamedRerouteUsage()
  NamedRerouteDeclaration(None=MaterialFunctionCall!(
    UV=MaterialFunctionCall!(
     UV_Input=@n2
     PolarUV_CenterOffsetG
     PolarUV_CenterOffsetR
     PolarUV_LenghtScale
     PolarUV_RadialScale
     PolarUV_UseTwist
     PolarUV_TwistOffset
     PolarUVTwistSpeed
     PolarUV_TwistRadius
    ) {materialFunction=/Game/Global/Materials/MaterialGym/Materials/MaterialElements/UVs/UVMainpulation/PolarUV/MF_Polar}.Result
    NoiseNormal
    NoiseTexture01
    NoiseTexture02
    DistortionTexture
    NoiseColorLight
    NoiseColorDarkOffset
    NoiseTint
    NoiseTiling
    NoiseSpeed
    DistortionTiling
    DistortionSpeed
   ) {materialFunction=/Game/Global/Materials/MaterialGym/Materials/MaterialElements/UVs/AnimatedNoise/MF_StormNoise}.Result) {name=NoiseX, nodeColor=<rgba 0 1 0.0859 1>}
  n3 = NamedRerouteUsage()
  NamedRerouteDeclaration(None=MaterialFunctionCall(UV=MaterialFunctionCall(UV_Input=@n3) {materialFunction=&MF_Polar}.Result) {materialFunction=&MF_StormNoise}.Result) {name=NoiseY, nodeColor=<rgba 1 0 0.2344 1>}
  n4 = NamedRerouteUsage()
  NamedRerouteDeclaration(None=MaterialFunctionCall(UV=MaterialFunctionCall(UV_Input=@n4) {materialFunction=&MF_Polar}.Result) {materialFunction=&MF_StormNoise}.Result) {name=NoiseZ, nodeColor=<rgba 0 0.5547 1 1>}
 ]
}
